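; Audio volume Blueprint that, going by its name, plays daytime music for the Kelp POI as the listener enters and leaves.
; On enter, it fades in the audio component, then once registers the default submix with Fort audio analysis and adds a POI Special event to the music event subsystem.
; On exit, it starts a matching fade-out, then, if registered, unregisters the submix, removes the event and immediately deactivates the component.
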(event OnListenerEnterEvent(AGV_ref)
  (Audio|Components|Audio|FadeIn (Variables|Default|GetAudio) 6.0 1.0 0.0 "Linear" :!class /Script/Engine.AudioComponent)
  (if (not (Variables|Default|GetbSubmixRegistered))
    (AudioAnalysis|RegisterDefaultSubmix (WorldSubsystems|GetFortAudioAnalysisSubsystem))
    (Variables|Default|SetbSubmixRegistered true)
    (Audio|AddEventbyComponent (GameInstanceSubsystems|GetMusicEventSubsystem) (Variables|Default|GetAudio) "(TagName=\"MusicEventSubsystem.Event.POI.Special\")" "(GameplayTags=)" 0 "")))

(event OnListenerExitEvent(AGV_ref)
  (bind _audio (Variables|Default|GetAudio))
  (Audio|Components|Audio|FadeOut (Variables|Default|GetAudio) 6.0 0.0 "Linear" :!class /Script/Engine.AudioComponent)
  (if (Variables|Default|GetbSubmixRegistered)
    (AudioAnalysis|UnRegisterDefaultSubmix (WorldSubsystems|GetFortAudioAnalysisSubsystem))
    (Variables|Default|SetbSubmixRegistered false)
    (Audio|RemoveEvent (GameInstanceSubsystems|GetMusicEventSubsystem) _audio)
    (Components|Activation|Deactivate _audio)))

(fn ConstructionScript)
