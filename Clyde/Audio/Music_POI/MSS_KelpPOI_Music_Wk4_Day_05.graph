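# MetaSound source that plays the Week 4 Day 05 KelpLinen POI music track as spatially attenuated stereo in the Fortnite music sound class.
# On play, one stereo Wave Player feeds the track straight to both outputs.
# The source is flagged looping and keeps playing when silent, but the player's Loop default holds a stray builder-result string.
# An Audio Bus input defaults to the music reactivity bus but connects to nothing in the graph.
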MetaSoundSource /Clyde/Audio/Music_POI/MSS_KelpPOI_Music_Wk4_Day_05
inputs [
 Audio Bus {type=AudioBusAsset}
 Day05 Music {type=WaveAsset}
 UE.Source.OnPlay {type=Trigger}
]
interfaces [UE.OutputFormat.Stereo:1.0, UE.Source:1.1]
outputs [
 "UE.OutputFormat.Stereo.Audio:0" {type=Audio}
 "UE.OutputFormat.Stereo.Audio:1" {type=Audio}
]
pages [Default {
  input_defaults {
   Audio Bus=&/AB_KelpLinen_Music_Reactivity
   Day05 Music=&/MX_IP_KelpLinen_Wk4_Day05_Bpm92
   UE.Source.OnPlay=null
  }
  nodes [WavePlayer_0 = UE.Wave Player:Stereo(Wave Asset=^Day05 Music, Play=^UE.Source.OnPlay) {also {defaults {Loop="<MetaSoundBuilderResult.SUCCEEDED: 0>"}}}]
  links [{dst="@UE.OutputFormat.Stereo.Audio:0", src=WavePlayer_0.Out Left}, {dst="@UE.OutputFormat.Stereo.Audio:1", src=WavePlayer_0.Out Right}]
 }]
properties {
 assetImportData=&self:AssetImportData
 attenuationSettings=&/SA_KelpLinen_POI_Music
 bLooping=true
 duration=10000
 maxDistance=20000
 numChannels=2
 outputFormat=Stereo
 sampleRate=48000
 soundClassObject=/Game/Sounds/FortSoundClassesAndModes/Fort_Music
 virtualizationMode=PlayWhenSilent
 soundSubmixSends [{minSendLevel=1}]
}
